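# Niagara boost trail effect for the IvoryCascadeLTE01 vehicle cosmetic, streaming backward from the car's boost sockets.
# A ribbon emitter spawns a ribbon per port, computing two phase-offset wave positions for its two renderers, effectively vertical since X/Y scales are zero.
# Curves over particle life shape wave amplitude, width, material parameters and inherited car movement.
# A second emitter adds mesh and sprite particles that grow, shrink and turn with steering input.
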NiagaraSystem /VehicleCosmetics/Boosts/Boost_IvoryCascadeLTE01/FX/NS_Boost_IvoryCascadeLTE01
dependencies {
 usedDataInterfaces [&NiagaraDataInterfaceCurve, &NiagaraDataInterfaceArrayFloat, &NiagaraDataInterfaceSkeletalMesh]
 usedDynamicInputs []
 usedModules [
  &self:SystemSpawnScript.SystemScriptSource.SystemScriptGraph.NiagaraNodeAssignment_0.SetVariables_#1
  &SystemState
 ]
}
emitters {
 NE_BlightShield_OPT {
  emitterUpdateScript [
   &EmitterState(Life Cycle Mode=::NewEnumerator1, Loop Duration=1, UseLoopDelay=false)
   &VC_FilteredSocketCount()
   &VC_RateQualityHandler(High Spawn Rate=60, Low Spawn Rate=30, Use Custom Rate?=true)
   &VC_SpawnRibbon_v1(Number of Ribbons=$Emitter.PortCount, SpawnRate=$Emitter.Spawn Rate)
  ]
  particleSpawnScript [
   &VC_PaintedVariantsOverride(
    DefaultColor=<rgba 1.722 0.39 0.01 1>
    R1 DefaultColor=<rgba 0.8438 0.5345 0.0552 1>
    R1 Gold:LinearColor=<rgba 0.9479 0.9479 0.9479 1>
    R2 DefaultColor:LinearColor=<rgba 0.8385 0.8385 0.8385 1>
    Use R1 Painted Overrides=true
    Use R2 Painted Overrides=true
   ) {enabled=false}
   &VC_InitializeRibbon_v1(
    Color=<rgba 1.722 0.39 0.01 1>
    Custom Offset=true
    Enable Skeletal Position=true
    Facing Axis:ENiagaraOrientationAxis=::ZAxis
    Lifetime=1.25
    Local Vector Offset=&AddVectorToPosition(Position=<-76 0 -2>, Vector=Linked() {linkedVariable {name=User.DTCarOffset, type=NiagaraPosition}})
    Offset Amount=1
    Ribbon Width=100
   )
   &AddVelocity(Velocity=<-800 0 0>) {module=AddVelocity001}
   /SetVariables(Particles.R2Color:LinearColor=<rgba 1 0.8337 0.4427 1>) {isSetParameters=true}
  ]
  particleUpdateScript [
   &ParticleState()
   &InheritSourceMovement(
    Applied Position Scale=&VectorFromFloat(Value=&FloatFromCurve(FloatCurve=DataInterface() {propertyValues {Curve {keys [
          {interpMode=RCIM_Cubic, tangentMode=RCTM_Break, time=0, value=1}
          {arriveTangent=-2.9025, interpMode=RCIM_Cubic, tangentMode=RCTM_Break, time=1, value=0}
         ]}}}))
    ApplyForce=false
    ApplyPosition=true
    ApplyPositionFalloff=false
   )
   /SetVariables(
    Particles.Ribbon_BlackColor:LinearColor=<rgba 0.9807 1 0.0778 1>
    Particles.WaveScaleCurve:NiagaraFloat=&FloatFromCurve(FloatCurve=DataInterface() {propertyValues {Curve {keys [
         {arriveTangent=11.2628, interpMode=RCIM_Cubic, leaveTangent=11.2628, tangentMode=RCTM_User, time=0.0092, value=0}
         {
          arriveTangent=0.7219
          arriveTangentWeight=0.0795
          interpMode=RCIM_Cubic
          leaveTangent=0.1806
          leaveTangentWeight=0.0696
          tangentMode=RCTM_Break
          tangentWeightMode=RCTWM_WeightedBoth
          time=0.125
          value=0.75
         }
         {arriveTangent=-0.5282, interpMode=RCIM_Cubic, leaveTangent=-0.5282, tangentMode=RCTM_User, time=0.45, value=0.35}
         {
          arriveTangent=-0.5881
          arriveTangentWeight=0.5973
          interpMode=RCIM_Cubic
          leaveTangent=-0.5881
          leaveTangentWeight=0.5973
          tangentMode=RCTM_User
          time=0.75
          value=0.15
         }
        ]}}})
   ) {isSetParameters=true}
   /SetVariables(
    Particles.CustomPosition_R1:NiagaraPosition=&AddVectorToPosition(
     Position=$Particles.Position
     Vector=&Multiply_VectorByFloat(
      Vector=&MakeVector(
       X=&Sine(
        Normalized Angle=&Add_Float(A=0, B=&Subtract_Float(A=$Emitter.NormalizedLoopAge, B=$Particles.NormalizedAge))
        Period=$User.WavePeriod
        Scale=&Multiply_Float(
         A=&SelectFloatFromArray(Direct Array Index=$Particles.SpawnGroup, Float Selection Array=DataInterface() {propertyValues {FloatData [0,0]}})
         B=$User.WaveMaxScale
        )
        Bias=0
       )
       Y=&Sine(
        Normalized Angle=&Add_Float(A=0, B=&Subtract_Float(A=$Emitter.NormalizedLoopAge, B=$Particles.NormalizedAge))
        Period=$User.WavePeriod
        Scale=&Multiply_Float(
         A=&SelectFloatFromArray(Direct Array Index=$Particles.SpawnGroup, Float Selection Array=DataInterface() {propertyValues {FloatData [0,0]}})
         B=$User.WaveMaxScale
        )
        Bias=0
       )
       Z=&Cosine(
        Normalized Angle=&Add_Float(A=0, B=&Subtract_Float(A=$Emitter.NormalizedLoopAge, B=$Particles.NormalizedAge))
        Period=$User.WavePeriod
        Scale=$User.WaveMaxScale
        Bias=0
       )
      )
      Float=$Particles.WaveScaleCurve
     )
    )
    Particles.CustomPosition_R2:NiagaraPosition=&AddVectorToPosition(
     Position=$Particles.Position
     Vector=&Multiply_VectorByFloat(
      Vector=&MakeVector(
       X=&Sine(
        Normalized Angle=&Add_Float(
         A=&Multiply_Float(A=$User.WavePeriod, B=0.5)
         B=&Subtract_Float(A=$Emitter.NormalizedLoopAge, B=$Particles.NormalizedAge)
        )
        Period=$User.WavePeriod
        Scale=&Multiply_Float(
         A=&SelectFloatFromArray(Direct Array Index=$Particles.SpawnGroup, Float Selection Array=DataInterface() {propertyValues {FloatData [0,0]}})
         B=$User.WaveMaxScale
        )
        Bias=0
       )
       Y=&Sine(
        Normalized Angle=&Add_Float(
         A=&Multiply_Float(A=$User.WavePeriod, B=0.5)
         B=&Subtract_Float(A=$Emitter.NormalizedLoopAge, B=$Particles.NormalizedAge)
        )
        Period=$User.WavePeriod
        Scale=&Multiply_Float(
         A=&SelectFloatFromArray(Direct Array Index=$Particles.SpawnGroup, Float Selection Array=DataInterface() {propertyValues {FloatData [0,0]}})
         B=$User.WaveMaxScale
        )
        Bias=0
       )
       Z=&Cosine(
        Normalized Angle=&Add_Float(
         A=&Multiply_Float(A=$User.WavePeriod, B=0.5)
         B=&Subtract_Float(A=$Emitter.NormalizedLoopAge, B=$Particles.NormalizedAge)
        )
        Period=$User.WavePeriod
        Scale=$User.WaveMaxScale
        Bias=0
       )
      )
      Float=$Particles.WaveScaleCurve
     )
    )
   ) {isSetParameters=true}
   /SetVariables(Particles.DynamicMP_R2:Vector4f=&MakeVector4(
     X=&FloatFromCurve(FloatCurve=DataInterface() {propertyValues {Curve {
         defaultValue=1
         keys [
          {tangentMode=RCTM_User, time=0, value=1}
          {arriveTangent=-0.0633, interpMode=RCIM_Cubic, leaveTangent=-0.0633, tangentMode=RCTM_User, time=0.65, value=1}
          {arriveTangent=-0.0699, interpMode=RCIM_Cubic, leaveTangent=-0.0699, tangentMode=RCTM_User, time=0.75, value=-2}
         ]
        }}})
     Y=0
     Z=0
     W=0
    )) {isSetParameters=true}
   &DynamicMaterialParameters(
    Index 0 Param 1=&FloatFromCurve(FloatCurve=DataInterface() {propertyValues {Curve {
        defaultValue=1
        keys [
         {arriveTangent=-0.0405, interpMode=RCIM_Cubic, leaveTangent=-0.0405, tangentMode=RCTM_User, time=0.45, value=1}
         {arriveTangent=-0.0201, interpMode=RCIM_Cubic, leaveTangent=-0.0201, tangentMode=RCTM_User, time=1, value=-2}
        ]
       }}})
    Index 1 Param 1=&FloatFromCurve(FloatCurve=DataInterface() {propertyValues {Curve {keys [<at 0 1>,<at 0.2 -0.8>,<at 0.5 -0.8>,<at 1 1>]}}})
    Index 1 Param 2=1
    Module.Index0Float0Write=true
    Module.Index0Float1Write=true
    Module.Index0Float2Write=true
    Module.Index0Float3Write=true
    Module.Index1Float0Write=true
    Module.Index1Float1Write=true
    Module.Index1Float2Write=true
    Module.Index1Float3Write=true
    Write Parameter Index 1=::NewEnumerator1
   )
   &SolveForcesAndVelocity(Clamp Velocity=false, Limit Acceleration=false)
   &ScaleRibbonWidth(Ribbon Width Scale=&FloatFromCurve(FloatCurve=DataInterface() {propertyValues {Curve {
        defaultValue=1
        keys [
         {
          arriveTangent=-0.2265
          arriveTangentWeight=0.1118
          interpMode=RCIM_Cubic
          leaveTangent=-0.2265
          leaveTangentWeight=0.1118
          tangentMode=RCTM_User
          tangentWeightMode=RCTWM_WeightedBoth
          time=0
          value=1
         }
         {arriveTangent=-0.4829, interpMode=RCIM_Cubic, leaveTangent=-0.4829, tangentMode=RCTM_User, time=0.3862, value=0.9365}
         {arriveTangent=-1.8813, interpMode=RCIM_Cubic, leaveTangent=-1.8813, tangentMode=RCTM_User, time=0.5846, value=0.6328}
         {arriveTangent=0.4473, interpMode=RCIM_Cubic, leaveTangent=0.4473, tangentMode=RCTM_User, time=0.75, value=0.5}
        ]
       }}}))
   &VC_UpdateRibbon_v1(Enable Skeletal Position=true)
  ]
  rendererClasses [&NiagaraRibbonRendererProperties]
  renderers [
   {rendererClass=&NiagaraRibbonRendererProperties}
   {rendererClass=&NiagaraRibbonRendererProperties, rendererIndex=1}
  ]
 }
 NE_VC_Boost_Sprite_Parent007 {
  emitterUpdateScript [
   &EmitterState()
   &VC_FilteredSocketCount()
   &SpawnRate(SpawnRate=&RandomRangeFloat(Minimum=10, Maximum=20), Use Spawn Probability=false)
  ]
  particleSpawnScript [
   &VC_PaintedVariantsOverride(DefaultColor=<rgba 0.4 0.4 0.4 1>)
   &InitializeParticle(
    Color Mode=::NewEnumerator0
    Lifetime=&MergeHelper_FloatSwitch(0=2, 1=1.5, 2=1.5, 3=?, Index=0, Num=::NewEnumerator1)
    Mesh Scale=<0.6 0.7 0.6>
    Mesh Scale Mode=::NewEnumerator4
    Sprite Rotation Mode=::NewEnumerator2
    UsePositionOffset=false
    Write Material Random=true
   )
   &PartitionParticles(Number Of Partitions=$Emitter.PortCount)
   &SkeletalMeshLocation(
    Bone / Socket Index=$Particles.PartitionParticles.Partition
    Bone / Socket Sampling Mode=::NewEnumerator5
    MaskSpawnGroup=false
    Mesh Sampling Type=::NewEnumerator6
    Sampled Position Offset=&AddVectorToPosition(Position=<-50 0 0>, Vector=Linked() {linkedVariable {name=User.DTCarOffset, type=NiagaraPosition}})
    Skeletal Mesh=$System.SkeletalMesh
   )
   &AddVelocity(Velocity=<-300 0 0>)
   &InitialMeshOrientation(Mesh Axis To Orient=<0 0 0>, Mesh Orientation Mode=::NewEnumerator2, Orientation Vector=<-1 0 0>, Use Rotation Vector=true) {module=InitialMeshOrientation001}
  ]
  particleUpdateScript [
   &ParticleState()
   &InheritSourceMovement(
    Applied Force Scale=<0.15 0.15 0.15>
    Applied Position Scale=&Multiply_VectorByFloat(
     Vector=<1 1 1>
     Float=&FloatFromCurve(FloatCurve=DataInterface() {propertyValues {Curve {keys [<at 0.5 1 RCIM_Cubic>,<at 1 0 RCIM_Cubic>]}}})
    )
    ApplyForce=true
    ApplyForceFalloff=false
    ApplyPosition=true
    ApplyPositionFalloff=false
    Module.LimitForceSpeed=false
   )
   &ScaleMeshSize(Scale Factor=&VectorFromFloat(Value=&FloatFromCurve(
      FloatCurve=DataInterface() {propertyValues {Curve {keys [
          <at 0 0.05>
          <at 0.2 0.3>
          {arriveTangent=-0.5455, interpMode=RCIM_Cubic, leaveTangent=-0.5455, time=0.5, value=0.2}
          {arriveTangent=0.0111, interpMode=RCIM_Cubic, leaveTangent=0.0111, tangentMode=RCTM_User, time=0.75, value=0}
          <at 1 0>
         ]}}}
      Scale Curve=2.5
     ))) {module=ScaleMeshSize001}
   &UpdateMeshOrientation(
    Facing Direction=&Lerp_Vector(
     A=<1 0 0>
     B=&MakeVector(X=0, Y=&Multiply_Float(A=$User.SteeringDir, B=-1), Z=0)
     Alpha=$Particles.NormalizedAge
    )
    Facing Interpolation=false
    Orientation Method=::NewEnumerator1
    Override Quat=false
    TransformFacing=false
    TransformUp=false
   )
   &SolveForcesAndVelocity(Clamp Velocity=false, Limit Acceleration=false)
  ]
  rendererClasses [&NiagaraMeshRendererProperties, &NiagaraSpriteRendererProperties]
  renderers [
   {rendererClass=&NiagaraMeshRendererProperties, rendererIndex=0}
   {rendererClass=&NiagaraSpriteRendererProperties, rendererIndex=1}
  ]
 }
}
summary {
 systemName=NS_Boost_IvoryCascadeLTE01
 emitters [
  {emitterName=NE_BlightShield_OPT, rendererClasses [&NiagaraRibbonRendererProperties]}
  {
   emitterName=NE_VC_Boost_Sprite_Parent007
   rendererClasses [&NiagaraMeshRendererProperties, &NiagaraSpriteRendererProperties]
  }
 ]
 userVariables [
  User.BackLightPower {type=NiagaraInt32, defaultValue=1}
  User.bIsFrontendPreview {type=NiagaraBool}
  User.bIsPainted {type=NiagaraBool}
  User.ColorVariantID {type=NiagaraInt32}
  User.DTCarOffset {type=NiagaraPosition, description="", defaultValue {}}
  User.DTCarOffset {type=Vector3f, description="", defaultValue {struct=Vector3f, value=<0 0 0>}}
  User.DTCarOffset001 {type=Vector3f, description="", defaultValue {struct=Vector3f, value=<0 0 0>}}
  User.GameModeID {type=NiagaraInt32}
  User.offset {type=NiagaraPosition, description="", defaultValue {}}
  User.offset {type=Vector3f, description="", defaultValue {struct=Vector3f, value=<0 0 0>}}
  User.offset001 {type=Vector3f, description="", defaultValue {struct=Vector3f, value=<0 0 0>}}
  User.PaintedColor {type=LinearColor, description="", defaultValue {struct=LinearColor, value=<rgba 1 1 1 1>}}
  User.PaintedScalar {type=NiagaraFloat, defaultValue=1, description=""}
  User.PortCount {type=NiagaraInt32, defaultValue=2}
  User.SingleBoostOffset {type=NiagaraPosition, description="", defaultValue {}}
  User.SingleBoostPos {type=NiagaraPosition, description="", defaultValue {}}
  User.SingleBoostPos {type=Vector3f, description="", defaultValue {struct=Vector3f, value=<0 0 0>}}
  User.SingleBoostPos001 {type=Vector3f, description="", defaultValue {struct=Vector3f, value=<0 0 0>}}
  User.SparksColorScale {type=NiagaraFloat, defaultValue=5, description=""}
  User.SteeringDir {type=NiagaraFloat, defaultValue=0, description=""}
  User.TextSwtch {type=NiagaraInt32}
  User.WaveMaxScale {type=NiagaraFloat, defaultValue=10, description=""}
  User.WavePeriod {type=NiagaraFloat, defaultValue=1, description=""}
 ]
}
